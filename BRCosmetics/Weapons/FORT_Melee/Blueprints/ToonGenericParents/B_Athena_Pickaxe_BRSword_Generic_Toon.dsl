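; Generic toon-shaded parent Blueprint for sword pickaxe cosmetics, extending its parent's weapon init and swing events.
; On cosmetic init in a frontend display, it gets or creates dynamic materials on main and offhand meshes, applying an inverse front-end LUT unless PBR_Shading is 1, marked as a temporary workaround.
; Each of the four swing events passes the instigator's montage section index to valid main and offhand swing VFX as SwingSectionIndex.
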(event EventOnInitWeaponCosmetics
  (bind _currentmid_0 (Variables|Default|GetCurrentMID_0))
  (bind _currentoffhandmid_0 (Variables|Default|GetCurrentOffhandMID_0))
  (|CallParentFunction)
  ;[1 1 1] Front End LUT, Remove after creating auto LUT tool
  (if (Reactive|Frontend|IsCosmeticObjectDisplayedinFrontendContext self)
    (Utilities|FlowControl|Sequence
      (:then_0
        (for (_array_element _array_index) (Rendering|Material|GetMaterials (Variables|Visual|GetWeaponMesh))
          (bind _asmaterial_instance_dynamic (Utilities|Casting|CastToMaterialInstanceDynamic _array_element)
            (:then
              (Variables|Default|SetCurrentMID_0 _asmaterial_instance_dynamic))
            (:CastFailed
              (bind _returnvalue (Rendering|Material|CreateDynamicMaterialInstance _array_element "None" "None" :!class /Script/Engine.KismetMaterialLibrary))
              (Variables|Default|SetCurrentMID_0 _returnvalue)
              (Rendering|Material|SetMaterial (Variables|Visual|GetWeaponMesh) _array_index _currentmid_0)))
          (bind _returnvalue_2 (Rendering|Material|GetScalarParameterValue _currentmid_0 "PBR_Shading" :!class /Script/Engine.MaterialInstanceDynamic))
          (if (not (== _returnvalue_2 1.0))
            (Rendering|Material|SetTextureParameterValue _currentmid_0 "LUT Volume Texture" "/Game/Packages/Fortress_SharedMaterials/Anime_Master_Material/VT_InverseLUT_FE.VT_InverseLUT_FE")
            (Rendering|Material|SetScalarParameterValue _currentmid_0 "LUT Multiplier" 1.0 :!class /Script/Engine.MaterialInstanceDynamic)
            (Rendering|Material|SetScalarParameterValue _currentmid_0 "Mobile LUT Brightness" 1.0 :!class /Script/Engine.MaterialInstanceDynamic)
            (Rendering|Material|SetScalarParameterValue _currentmid_0 "use 33 Resolution" 0.0 :!class /Script/Engine.MaterialInstanceDynamic))))
      (:then_1
        (for (_array_element_2 _array_index_2) (Rendering|Material|GetMaterials (Variables|Visual|GetWeaponMeshOffhand))
          (bind _asmaterial_instance_dynamic_2 (Utilities|Casting|CastToMaterialInstanceDynamic _array_element_2)
            (:then
              (Variables|Default|SetCurrentOffhandMID_0 _asmaterial_instance_dynamic_2))
            (:CastFailed
              (bind _returnvalue_3 (Rendering|Material|CreateDynamicMaterialInstance _array_element_2 "None" "None" :!class /Script/Engine.KismetMaterialLibrary))
              (Variables|Default|SetCurrentOffhandMID_0 _returnvalue_3)
              (Rendering|Material|SetMaterial (Variables|Visual|GetWeaponMeshOffhand) _array_index_2 _currentoffhandmid_0)))
          (bind _returnvalue_4 (Rendering|Material|GetScalarParameterValue _currentoffhandmid_0 "PBR_Shading" :!class /Script/Engine.MaterialInstanceDynamic))
          (if (not (== _returnvalue_4 1.0))
            (Rendering|Material|SetTextureParameterValue _currentoffhandmid_0 "LUT Volume Texture" "/Game/Packages/Fortress_SharedMaterials/Anime_Master_Material/VT_InverseLUT_FE.VT_InverseLUT_FE")
            (Rendering|Material|SetScalarParameterValue _currentoffhandmid_0 "LUT Multiplier" 1.0 :!class /Script/Engine.MaterialInstanceDynamic)
            (Rendering|Material|SetScalarParameterValue _currentoffhandmid_0 "Mobile LUT Brightness" 1.0 :!class /Script/Engine.MaterialInstanceDynamic)
            (Rendering|Material|SetScalarParameterValue _currentoffhandmid_0 "use 33 Resolution" 0.0 :!class /Script/Engine.MaterialInstanceDynamic)))))))

(event EventSwingRight
  (bind _returnvalue (CallFunction|GetSwingOffhandVFXComponent))
  (bind _returnvalue_2 (CallFunction|GetSwingVFXComponent))
  (|CallParentFunction)
  (bind _section_index (Utilities|SetupSwingMontageSectionIndex (Variables|Actor|GetInstigator))
    (:Out
      (Utilities|FlowControl|Sequence
        (:then_0
          (Utilities|IsValid _returnvalue_2
            (:Is%20Valid
              (Effects|Components|ParticleSystem|SetIntParameter _returnvalue_2 "SwingSectionIndex" _section_index))))
        (:then_1
          (Utilities|IsValid _returnvalue
            (:Is%20Valid
              (Effects|Components|ParticleSystem|SetIntParameter _returnvalue "SwingSectionIndex" _section_index))))))))

(event EventSwingRight2
  (bind _returnvalue (CallFunction|GetSwingOffhandVFXComponent))
  (bind _returnvalue_2 (CallFunction|GetSwingVFXComponent))
  (|CallParentFunction)
  (bind _section_index (Utilities|SetupSwingMontageSectionIndex (Variables|Actor|GetInstigator))
    (:Out
      (Utilities|FlowControl|Sequence
        (:then_0
          (Utilities|IsValid _returnvalue_2
            (:Is%20Valid
              (Effects|Components|ParticleSystem|SetIntParameter _returnvalue_2 "SwingSectionIndex" _section_index))))
        (:then_1
          (Utilities|IsValid _returnvalue
            (:Is%20Valid
              (Effects|Components|ParticleSystem|SetIntParameter _returnvalue "SwingSectionIndex" _section_index))))))))

(event EventSwingLeft
  (bind _returnvalue (CallFunction|GetSwingOffhandVFXComponent))
  (bind _returnvalue_2 (CallFunction|GetSwingVFXComponent))
  (|CallParentFunction)
  (bind _section_index (Utilities|SetupSwingMontageSectionIndex (Variables|Actor|GetInstigator))
    (:Out
      (Utilities|FlowControl|Sequence
        (:then_0
          (Utilities|IsValid _returnvalue_2
            (:Is%20Valid
              (Effects|Components|ParticleSystem|SetIntParameter _returnvalue_2 "SwingSectionIndex" _section_index))))
        (:then_1
          (Utilities|IsValid _returnvalue
            (:Is%20Valid
              (Effects|Components|ParticleSystem|SetIntParameter _returnvalue "SwingSectionIndex" _section_index))))))))

(event EventSwingLeft2
  (bind _returnvalue (CallFunction|GetSwingOffhandVFXComponent))
  (bind _returnvalue_2 (CallFunction|GetSwingVFXComponent))
  (|CallParentFunction)
  (bind _section_index (Utilities|SetupSwingMontageSectionIndex (Variables|Actor|GetInstigator))
    (:Out
      (Utilities|FlowControl|Sequence
        (:then_0
          (Utilities|IsValid _returnvalue_2
            (:Is%20Valid
              (Effects|Components|ParticleSystem|SetIntParameter _returnvalue_2 "SwingSectionIndex" _section_index))))
        (:then_1
          (Utilities|IsValid _returnvalue
            (:Is%20Valid
              (Effects|Components|ParticleSystem|SetIntParameter _returnvalue "SwingSectionIndex" _section_index))))))))

(fn ConstructionScript
  (|CallParentFunction))
