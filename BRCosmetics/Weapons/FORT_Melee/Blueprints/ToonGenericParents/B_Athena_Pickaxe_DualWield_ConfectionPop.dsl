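; Cosmetic Blueprint for the ConfectionPop dual-wield pickaxe, handling toon-styled trail and impact effects.
; The pickaxe particle variant sets IsToon/IsNotToon on trails and impact effects.
; Impacts also get a BeepBoop value from the last swing side, or alternating on a short delay in frontend or preview.
; On BeginPlay in the frontend, both meshes' non-PBR materials become dynamic instances with an inverse LUT, a fix marked temporary.
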(event EventSwingLeft_Common
  ;[1 0.4253 0.8423] Swing Left Right for Impact
  (|CallParentFunction)
  (Variables|Default|SetIsRight false))

(event EventSwingRight_Common
  ;[1 0.4253 0.8423] Swing Left Right for Impact
  (|CallParentFunction)
  (Variables|Default|SetIsRight true))

(event EventOnInitWeaponCosmetics
  (bind _returnvalue (CallFunction|GetAnimTrailsPSC))
  ;[1 0.2887 0.2488] Setting Animtrail Variant
  (|CallParentFunction)
  (bind _index (Class|FortCosmeticFunctionLibrary|%28Deprecated%29GetVariantChannelActiveIndexForSlot "(TagName=\"LoadoutSlot.Pickaxe\")" (CallFunction|GetCosmeticData) "(TagName=\"Cosmetics.Variant.Channel.Particle\")" (Utilities|Array|MakeArray "(TagName=\"Cosmetics.Variant.Property.Particle0\")" "(TagName=\"Cosmetics.Variant.Property.Particle1\")" "(TagName=\"Cosmetics.Variant.Property.Particle2\")")))
  (if (!= _index 0)
    (Effects|Components|ParticleSystem|SetBoolParameter _returnvalue "IsToon" false)
    (Effects|Components|ParticleSystem|SetBoolParameter _returnvalue "IsNotToon" true)
    (else
      (Effects|Components|ParticleSystem|SetBoolParameter _returnvalue "IsToon" true)
      (Effects|Components|ParticleSystem|SetBoolParameter _returnvalue "IsNotToon" false))))

(event FortWeapon|EventOnPlayImpactFX
  (params (HitResult :type Hit%20Result%20Structure%20%28by%20ref%29) (ImpactPhysicalSurface :type EPhysicalSurface%20Enum :default "SurfaceType_Default") (SpawnedPSC :type FXSystem%20Component%20Object%20Reference))
  (bind _frontendint (Variables|Default|GetFrontEndInt))
  ;[1 0.6602 0.2944] Impact FX
  (|CallParentFunction HitResult ImpactPhysicalSurface SpawnedPSC)
  (Utilities|IsValid SpawnedPSC
    (:Is%20Valid
      (Utilities|FlowControl|Sequence
        (:then_0
          (if (or (Reactive|Frontend|IsCosmeticObjectDisplayedinFrontendContext 0) (Reactive|Frontend|IsCosmeticObjectDisplayedasPreview 0))
            ;[0.3277 0.5451 1] FrontEndReactivity
            (switch int (Variables|Default|GetFrontEndInt)
              (:0
                (Effects|Components|ParticleSystem|SetIntParameter SpawnedPSC "BeepBoop" _frontendint)
                (Utilities|FlowControl|Delay 0.2)
                (Variables|Default|SetFrontEndInt 1))
              (:1
                (Effects|Components|ParticleSystem|SetIntParameter SpawnedPSC "BeepBoop" _frontendint)
                (Utilities|FlowControl|Delay 0.2)
                (Variables|Default|SetFrontEndInt 0)))
            (elif (Variables|Default|GetIsRight)
              ;[0.3146 1 0.5848] LeftRightSwinging
              (Effects|Components|ParticleSystem|SetIntParameter SpawnedPSC "BeepBoop" 0)
              (else
                (Effects|Components|ParticleSystem|SetIntParameter SpawnedPSC "BeepBoop" 1)))))
        (:then_1
          (bind _asniagara_particle_system_component (Utilities|Casting|CastToNiagaraComponent SpawnedPSC))
          ;[0.657 1 0.3538] Setting Impact Variants
          (bind _index (Class|FortCosmeticFunctionLibrary|%28Deprecated%29GetVariantChannelActiveIndexForSlot "(TagName=\"LoadoutSlot.Pickaxe\")" (CallFunction|GetCosmeticData) "(TagName=\"Cosmetics.Variant.Channel.Particle\")" (Utilities|Array|MakeArray "(TagName=\"Cosmetics.Variant.Property.Particle0\")" "(TagName=\"Cosmetics.Variant.Property.Particle1\")" "(TagName=\"Cosmetics.Variant.Property.Particle2\")")))
          (if (!= _index 0)
            (Effects|Components|ParticleSystem|SetBoolParameter _asniagara_particle_system_component "IsToon" false)
            (Effects|Components|ParticleSystem|SetBoolParameter _asniagara_particle_system_component "IsNotToon" true)
            (else
              (Effects|Components|ParticleSystem|SetBoolParameter _asniagara_particle_system_component "IsToon" true)
              (Effects|Components|ParticleSystem|SetBoolParameter _asniagara_particle_system_component "IsNotToon" false))))))))

(event EventBeginPlay
  (bind _currentmid (Variables|Default|GetCurrentMID))
  (bind _currentoffhandmid (Variables|Default|GetCurrentOffhandMID))
  (|CallParentFunction)
  ;[1 1 1] Front End LUT, Remove after creating auto LUT tool
  (Utilities|FlowControl|Delay 0.1)
  (if (Reactive|Frontend|IsCosmeticObjectDisplayedinFrontendContext self)
    (Utilities|FlowControl|Sequence
      (:then_0
        (for (_array_element _array_index) (Rendering|Material|GetMaterials (Variables|Visual|GetWeaponMesh))
          (bind _asmaterial_instance_dynamic (Utilities|Casting|CastToMaterialInstanceDynamic _array_element)
            (:then
              (Variables|Default|SetCurrentMID _asmaterial_instance_dynamic))
            (:CastFailed
              (bind _returnvalue (Rendering|Material|CreateDynamicMaterialInstance _array_element "None" "None" :!class /Script/Engine.KismetMaterialLibrary))
              (Variables|Default|SetCurrentMID _returnvalue)
              (Rendering|Material|SetMaterial (Variables|Visual|GetWeaponMesh) _array_index _currentmid)))
          (bind _returnvalue_2 (Rendering|Material|GetScalarParameterValue _currentmid "PBR_Shading" :!class /Script/Engine.MaterialInstanceDynamic))
          (if (not (== _returnvalue_2 1.0))
            (Rendering|Material|SetTextureParameterValue _currentmid "LUT Volume Texture" "/Game/Packages/Fortress_SharedMaterials/Anime_Master_Material/VT_InverseLUT_FE.VT_InverseLUT_FE")
            (Rendering|Material|SetScalarParameterValue _currentmid "LUT Multiplier" 1.0 :!class /Script/Engine.MaterialInstanceDynamic)
            (Rendering|Material|SetScalarParameterValue _currentmid "Mobile LUT Brightness" 1.0 :!class /Script/Engine.MaterialInstanceDynamic)
            (Rendering|Material|SetScalarParameterValue _currentmid "use 33 Resolution" 0.0 :!class /Script/Engine.MaterialInstanceDynamic))))
      (:then_1
        (for (_array_element_2 _array_index_2) (Rendering|Material|GetMaterials (Variables|Visual|GetWeaponMeshOffhand))
          (bind _asmaterial_instance_dynamic_2 (Utilities|Casting|CastToMaterialInstanceDynamic _array_element_2)
            (:then
              (Variables|Default|SetCurrentOffhandMID _asmaterial_instance_dynamic_2))
            (:CastFailed
              (bind _returnvalue_3 (Rendering|Material|CreateDynamicMaterialInstance _array_element_2 "None" "None" :!class /Script/Engine.KismetMaterialLibrary))
              (Variables|Default|SetCurrentOffhandMID _returnvalue_3)
              (Rendering|Material|SetMaterial (Variables|Visual|GetWeaponMeshOffhand) _array_index_2 _currentoffhandmid)))
          (bind _returnvalue_4 (Rendering|Material|GetScalarParameterValue _currentoffhandmid "PBR_Shading" :!class /Script/Engine.MaterialInstanceDynamic))
          (if (not (== _returnvalue_4 1.0))
            (Rendering|Material|SetTextureParameterValue _currentoffhandmid "LUT Volume Texture" "/Game/Packages/Fortress_SharedMaterials/Anime_Master_Material/VT_InverseLUT_FE.VT_InverseLUT_FE")
            (Rendering|Material|SetScalarParameterValue _currentoffhandmid "LUT Multiplier" 1.0 :!class /Script/Engine.MaterialInstanceDynamic)
            (Rendering|Material|SetScalarParameterValue _currentoffhandmid "Mobile LUT Brightness" 1.0 :!class /Script/Engine.MaterialInstanceDynamic)
            (Rendering|Material|SetScalarParameterValue _currentoffhandmid "use 33 Resolution" 0.0 :!class /Script/Engine.MaterialInstanceDynamic)))))))

(fn ConstructionScript
  (|CallParentFunction))
