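; Generic parent Blueprint for toon-shaded dual-wield pickaxe cosmetics that corrects their color when displayed in the frontend.
; On weapon cosmetic init it runs parent logic, then, in frontend context only, gets or creates dynamic material instances on main and offhand meshes.
; Materials whose PBR_Shading is not 1 receive an inverse frontend LUT texture and related LUT scalar settings.
; A graph comment marks this as temporary until an auto LUT tool exists.
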(event EventOnInitWeaponCosmetics
  (bind _currentmid (Variables|Default|GetCurrentMID))
  (bind _currentoffhandmid (Variables|Default|GetCurrentOffhandMID))
  (|CallParentFunction)
  ;[1 1 1] Front End LUT, Remove after creating auto LUT tool
  (if (Reactive|Frontend|IsCosmeticObjectDisplayedinFrontendContext self)
    (Utilities|FlowControl|Sequence
      (:then_0
        (for (_array_element _array_index) (Rendering|Material|GetMaterials (Variables|Visual|GetWeaponMesh))
          (bind _asmaterial_instance_dynamic (Utilities|Casting|CastToMaterialInstanceDynamic _array_element)
            (:then
              (Variables|Default|SetCurrentMID _asmaterial_instance_dynamic))
            (:CastFailed
              (bind _returnvalue (Rendering|Material|CreateDynamicMaterialInstance _array_element "None" "None" :!class /Script/Engine.KismetMaterialLibrary))
              (Variables|Default|SetCurrentMID _returnvalue)
              (Rendering|Material|SetMaterial (Variables|Visual|GetWeaponMesh) _array_index _currentmid)))
          (bind _returnvalue_2 (Rendering|Material|GetScalarParameterValue _currentmid "PBR_Shading" :!class /Script/Engine.MaterialInstanceDynamic))
          (if (not (== _returnvalue_2 1.0))
            (Rendering|Material|SetTextureParameterValue _currentmid "LUT Volume Texture" "/Game/Packages/Fortress_SharedMaterials/Anime_Master_Material/VT_InverseLUT_FE.VT_InverseLUT_FE")
            (Rendering|Material|SetScalarParameterValue _currentmid "LUT Multiplier" 1.0 :!class /Script/Engine.MaterialInstanceDynamic)
            (Rendering|Material|SetScalarParameterValue _currentmid "Mobile LUT Brightness" 1.0 :!class /Script/Engine.MaterialInstanceDynamic)
            (Rendering|Material|SetScalarParameterValue _currentmid "use 33 Resolution" 0.0 :!class /Script/Engine.MaterialInstanceDynamic))))
      (:then_1
        (for (_array_element_2 _array_index_2) (Rendering|Material|GetMaterials (Variables|Visual|GetWeaponMeshOffhand))
          (bind _asmaterial_instance_dynamic_2 (Utilities|Casting|CastToMaterialInstanceDynamic _array_element_2)
            (:then
              (Variables|Default|SetCurrentOffhandMID _asmaterial_instance_dynamic_2))
            (:CastFailed
              (bind _returnvalue_3 (Rendering|Material|CreateDynamicMaterialInstance _array_element_2 "None" "None" :!class /Script/Engine.KismetMaterialLibrary))
              (Variables|Default|SetCurrentOffhandMID _returnvalue_3)
              (Rendering|Material|SetMaterial (Variables|Visual|GetWeaponMeshOffhand) _array_index_2 _currentoffhandmid)))
          (bind _returnvalue_4 (Rendering|Material|GetScalarParameterValue _currentoffhandmid "PBR_Shading" :!class /Script/Engine.MaterialInstanceDynamic))
          (if (not (== _returnvalue_4 1.0))
            (Rendering|Material|SetTextureParameterValue _currentoffhandmid "LUT Volume Texture" "/Game/Packages/Fortress_SharedMaterials/Anime_Master_Material/VT_InverseLUT_FE.VT_InverseLUT_FE")
            (Rendering|Material|SetScalarParameterValue _currentoffhandmid "LUT Multiplier" 1.0 :!class /Script/Engine.MaterialInstanceDynamic)
            (Rendering|Material|SetScalarParameterValue _currentoffhandmid "Mobile LUT Brightness" 1.0 :!class /Script/Engine.MaterialInstanceDynamic)
            (Rendering|Material|SetScalarParameterValue _currentoffhandmid "use 33 Resolution" 0.0 :!class /Script/Engine.MaterialInstanceDynamic)))))))

(fn ConstructionScript
  (|CallParentFunction))
